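# Default Lit material for the zipline gadget, built to spin parts of the motor mesh in the shader with diffuse, normal and mask textures.
# RotateAboutAxis turns vertices around local X, with angle from periodic Time, a RotationSpeed parameter and red vertex color.
# The pivot is raw RGB from a Pivot Painter 2 texture on UV1, plus an offset parameter, transformed to world space.
# FixRotateAboutAxisNormals applies the same rotation to the transformed normal map.
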Material /Game/Gadgets/Zipline/Materials/M_Zipline
graph {
 properties {pixelDepthOffsetMode=PDOM_Legacy, previewMesh=/Game/Gadgets/Zipline/Meshes/SM_Zipline_Motor, usedShadingModels=Default Lit}
 nodes [
  n1 = TextureSample(UVs=TextureCoordinate() {coordinateIndex=1}) {samplerType=SAMPLERTYPE_LinearColor, texture=/Game/Gadgets/Zipline/Textures/SM_Zipline_Motor_Detached_007_rgb_PivotPos_a_ParentIndexInt_UV_2}
  MaterialFunctionCall!(Position RGB=@n1.RGB) {materialFunction=/Engine/Functions/Engine_MaterialFunctions02/PivotPainter2/ms_PivotPainter2_DecodePostion}
  n2 = Transform(None=Constant3Vector() {constant=<rgba 1 0 0 0>}) {transformSourceType=TRANSFORMSOURCE_Local}
  n3 = Multiply(
   A=Multiply(A=Time() {bOverride_Period=true, period=1}, B=ScalarParameter() {defaultValue=-1, parameterName=RotationSpeed})
   B=VertexColor().R
  )
  MaterialFunctionCall(
   Rotation Axis=@n2
   Rotation Angle=@n3
   World Space Vertex Normals=Transform(None=TextureSample() {samplerType=SAMPLERTYPE_Normal, texture=/Game/Gadgets/Zipline/Textures/T_Zipline_N}.RGB)
  ) {materialFunction=&FixRotateAboutAxisNormals}
  n4 = Add(A=@n1.RGB, B=VectorParameter() {parameterName=offset, defaultValue=<rgba 0 -1 0 1>}.RGB)
  RotateAboutAxis(
   NormalizedRotationAxis=@n2
   RotationAngle=@n3
   PivotPoint=TransformPosition(None=@n4) {transformType=TRANSFORMPOSSOURCE_World}
   Position=WorldPosition().XYZ
  )
  TextureSample() {texture=/Game/Gadgets/Zipline/Textures/T_Zipline_D}
  TextureSample() {samplerType=SAMPLERTYPE_Masks, texture=/Game/Gadgets/Zipline/Textures/T_Zipline_S}
  TransformPosition(None=@n1.RGB) {transformType=TRANSFORMPOSSOURCE_World}
  Transform(None=@n4) {transformSourceType=TRANSFORMSOURCE_Local}
 ]
}
