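; Zipline harness motor actor that attaches to a riding player, providing orientation, audio and VFX.
; Timers realign the motor to the environmental or dynamic zipline spline and repeat enemy/neutral sound indicators.
; Another timer reads zipline speed factor to trigger local braking and downhill gameplay cues; uphill feedback is only debug prints.
; Cosmetic spawn/destroy effects bracket the looping travel cue, while non-cosmetic instances bind player death so the authority expires the actor.
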(event Custom|Start%20Travel%20Audio
  (bind _attachedplayer (Variables|Default|GetAttachedPlayer))
  ;[0 0 0] Audio logic
  (Projectile|AddGameplayCue%28Looping%29Local _attachedplayer (Variables|Default|GetGCNTag_Travel) (Ability|GameplayCue|MakeGameplayCueParameters 0.0))
  (if (not (Pawn|IsLocallyViewed _attachedplayer))
    (Fortnite|ShowSoundIndicatorLocalforTeam :Location (Transformation|GetActorLocation _attachedplayer) :MaxAudibleDistance 0.0 :IndicatorType "Generic" :Team 0 :Affiliations (Utilities|Array|MakeArray "Hostile") :Tint "(R=0.000000,G=0.000000,B=0.000000,A=0.000000)" :OverrideIcon "/Game/UI/Foundation/Textures/Icons/Navigation/zipline_visualizer_icon_32.zipline_visualizer_icon_32" :SoundBase (Variables|Default|GetAttachSound) :AttenuationMaxRangeMultiplier 1.0 :AttenuationIntensityMultiplier 1.0)))

(event Custom|Stop%20Travel%20Audio
  (bind _attachedplayer (Variables|Default|GetAttachedPlayer))
  ;[0 0 0] Audio logic
  (Projectile|RemoveGameplayCue%28Looping%29Local _attachedplayer (Variables|Default|GetGCNTag_Travel) (Ability|GameplayCue|MakeGameplayCueParameters 0.0))
  (if (not (Pawn|IsLocallyViewed _attachedplayer))
    (Fortnite|ShowSoundIndicatorLocalforTeam :Location (Transformation|GetActorLocation _attachedplayer) :MaxAudibleDistance 0.0 :IndicatorType "Generic" :Team 0 :Affiliations (Utilities|Array|MakeArray "Hostile") :Tint "(R=0.000000,G=0.000000,B=0.000000,A=0.000000)" :OverrideIcon "/Game/UI/Foundation/Textures/Icons/Navigation/zipline_visualizer_icon_32.zipline_visualizer_icon_32" :SoundBase (Variables|Default|GetDetachSound) :AttenuationMaxRangeMultiplier 1.0 :AttenuationIntensityMultiplier 1.0)))

(event EventBeginPlay
  (Networking|CanExecuteCosmeticEvents
    (:True
      (bind _returnvalue (Effects|SpawnEmitterAttached :EmitterTemplate (Variables|Default|GetSpawnFX) :AttachToComponent (Variables|Default|GetSM_Zipline_Motor) :AttachPointName "None" :Scale "1.000000,1.000000,1.000000" :LocationType "SnapToTarget" :bAutoDestroy true :PoolingMethod "None" :bAutoActivate true)) ; Play spawn VFX
      (Rendering|SetTranslucentSortPriority _returnvalue -1)
      (CallFunction|StartTravelAudio)
      (Transformation|SetWorldRotation (Variables|Default|GetSM_Zipline_Magnet) (Math|Rotator|MakeRotator 0.0 0.0 (.yaw (Transformation|GetWorldRotation (Variables|Default|GetSM_Zipline_Motor)))) false false)
      (bind _returnvalue_2 (Utilities|Time|SetTimerbyEvent (delegate Custom|Chime%20Visual) (Variables|Default|GetZiplineChimeVisualUpdate) true))
      (Variables|Default|SetChimeTimer _returnvalue_2)
      (Utilities|Time|SetTimerbyEvent (delegate Custom|UpdateParamsForAudioAndVFX) 0.12 true))
    (:False
      (EventDispatchers|BindEventtoOnDied (Variables|Default|GetAttachedPlayer) (delegate Custom|HandleOnDied)))))

(event Custom|Chime%20Visual
  (Fortnite|ShowSoundIndicatorLocalforTeam :TrackedActor (Variables|Default|GetAttachedPlayer) :Location (Transformation|GetActorLocation) :MaxAudibleDistance (Variables|/Script/Engine.SoundBase|GetMaxDistance (Variables|Default|GetTravelSound)) :IndicatorType "Vehicle" :Team (Team|GetActorTeam (Variables|Default|GetAttachedPlayer)) :Affiliations (Utilities|Array|MakeArray "Hostile" "Neutral") :Instigator self :Tint "(R=0.000000,G=0.000000,B=0.000000,A=0.000000)" :OverrideIcon "/Game/UI/Foundation/Textures/Icons/Navigation/zipline_visualizer_icon_32.zipline_visualizer_icon_32" :SoundBase (Variables|Default|GetTravelSound) :AttenuationMaxRangeMultiplier 0.5 :AttenuationIntensityMultiplier 1.0))

(event Custom|HandleOnDied
  (params (DamagedActor :type Actor%20Object%20Reference) (Damage :type Float%20%28single-precision%29) (InstigatedBy :type Controller%20Object%20Reference) (DamageCauser :type Actor%20Object%20Reference) (HitLocation :type Vector) (FHitComponent :type Primitive%20Component%20Object%20Reference) (BoneName :type Name) (Momentum :type Vector))
  ;[0 0 0] Kill the motor when the player dies
  (Networking|SwitchHasAuthority
    (:Authority
      (Actor|SetLifeSpan 0.2))))

(event Custom|UpdateParamsForAudioAndVFX
  (CallFunction|UpdateAudioAndVFXParams))

(event EventDestroyed
  (Networking|CanExecuteCosmeticEvents
    (:True
      (bind _returnvalue (Effects|SpawnEmitteratLocation (select (Variables|Default|GetHasHitAnotherPlayer) (Variables|Default|GetCollideDestroyVFX) (Variables|Default|GetDestroyFX)) (.location (Transformation|GetWorldTransform (Variables|Default|GetSM_Zipline_Motor))) (.rotation (Transformation|GetWorldTransform (Variables|Default|GetSM_Zipline_Motor))))) ; Play destruction VFX
      (Rendering|SetTranslucentSortPriority _returnvalue -1)
      (CallFunction|StopTravelAudio)
      (Utilities|Time|ClearandInvalidateTimerbyHandle (Variables|Default|GetMotorUpdateTimer)) ; probably not necessary
      (CallFunction|CleanuponDownhillGCN))))

(event Custom|UpdateMotorToSpline
  ;[0 0 0] Rotate To Match Zipline
  (Variables|Default|SetZiplineDirection (Spline|FindDirectionClosesttoWorldLocation (Variables|/Script/FortniteGame.FortAthenaSplineZipline|GetZiplineSpline (Variables|Default|GetEnvZiplineSpline)) (Transformation|GetWorldLocation (Variables|Default|GetSM_Zipline_Motor)) "World"))
  (CallFunction|RotatetoZiplineDirection))

(event Custom|RotateToZiplineDirection
  ;[0 0 0] Rotate To Match Zipline
  (Transformation|SetRelativeRotation (Variables|Default|GetSM_Zipline_Motor) (Math|Rotator|MakeRotator (- (.roll (Math|Rotator|MakeRotfromZ (Variables|Default|GetZiplineDirection))) 2.7e+02) (.pitch (Math|Rotator|MakeRotfromZ (Variables|Default|GetZiplineDirection))) (.yaw (Math|Rotator|MakeRotfromZ (Variables|Default|GetZiplineDirection)))) false false))

(event Custom|OrientMotorForZipline
  (params (Zipline :type Fort%20Athena%20Zipline%20Base%20Object%20Reference) (DesiredZiplineOffset :type Vector))
  ;[0 0 0] Rotate To Match Zipline
  (Variables|Default|SetCurrentDesired%20Zipline%20Offset DesiredZiplineOffset)
  (Transformation|AttachActorToComponent (select (Variables|Default|GetUseMeshAttachment) (Variables|/Script/Engine.Character|GetMesh (Variables|Default|GetAttachedPlayer)) (Variables|/Script/Engine.Actor|GetRootComponent (Variables|Default|GetAttachedPlayer))) "0" "SnapToTarget" "KeepWorld" "KeepWorld")
  (CallFunction|UpdateZiplineOffsetLocation)
  (bind _asbp_athena_environmental_zip_line_spline (Utilities|Casting|CastToBP_Athena_Environmental_ZipLine_Spline Zipline)
    (:then
      (Variables|Default|SetEnvZiplineSpline _asbp_athena_environmental_zip_line_spline)
      (bind _returnvalue (Utilities|Time|SetTimerbyEvent (delegate Custom|UpdateMotorToSpline) (Variables|Default|GetSplineRotationUpdateSeconds) true))
      (Variables|Default|SetMotorUpdateTimer _returnvalue)
      (Variables|Default|SetZiplineDirection (Spline|FindDirectionClosesttoWorldLocation (Variables|/Script/FortniteGame.FortAthenaSplineZipline|GetZiplineSpline (Variables|Default|GetEnvZiplineSpline)) (Transformation|GetWorldLocation (Variables|Default|GetSM_Zipline_Motor)) "World")))
    (:CastFailed
      ;;[1 1 1] Adding this Dynamic zipline check for the Zipline Deployer Plugin. TODO:  DaveM: clean this up
      (bind _asprototype_bp_athena_dynamic_zip_line_spline (Utilities|Casting|CastToPROTOTYPE_BP_Athena_Dynamic_ZipLine_Spline Zipline)
        (:then
          (Variables|Default|SetDynamicZiplineSpline _asprototype_bp_athena_dynamic_zip_line_spline)
          (bind _returnvalue_2 (Utilities|Time|SetTimerbyEvent (delegate Custom|UpdateMotorToDynSpline) (Variables|Default|GetSplineRotationUpdateSeconds) true))
          (Variables|Default|SetMotorUpdateTimer _returnvalue_2)
          (Variables|Default|SetZiplineDirection (Spline|FindDirectionClosesttoWorldLocation (Variables|/Script/FortniteGame.FortAthenaSplineZipline|GetZiplineSpline (Variables|Default|GetDynamicZiplineSpline)) (Transformation|GetWorldLocation (Variables|Default|GetSM_Zipline_Motor)) "World")))
        (:CastFailed
          (Development|PrintString "New Zipline type without orientation code in BP_Zipline_Athena_Harness")))))
  (CallFunction|RotatetoZiplineDirection))

(event Custom|UpdateMotorToDynSpline
  ;[0 0 0] Rotate To Match Zipline
  ;;[1 1 1] Adding this Dynamic zipline check for the Zipline Deployer Plugin. TODO:  DaveM: clean this up
  (Variables|Default|SetZiplineDirection (Spline|FindDirectionClosesttoWorldLocation (Variables|/Script/FortniteGame.FortAthenaSplineZipline|GetZiplineSpline (Variables|Default|GetDynamicZiplineSpline)) (Transformation|GetWorldLocation (Variables|Default|GetSM_Zipline_Motor)) "World"))
  (CallFunction|RotatetoZiplineDirection))

(event Custom|UpdateZiplineOffsetLocation
  ;[0 0 0] Rotate To Match Zipline
  (Transformation|SetRelativeLocation (Variables|Default|GetSM_Zipline_Motor) (Math|Vector|vector+vector (Variables|Default|GetCurrentDesired%20Zipline%20Offset) (Math|Vector|SelectVector "0, 0,295.000000" "0, 0,220.000000" (Variables|Default|GetUseMeshAttachment)) (Zipline|GetZiplineMotorOffset (Variables|Default|GetAttachedPlayer))) false false))

(event Custom|Cleanup%20On%20DownhillGCN
  ;[1 1 1] Cleanup on Destroyed
  (if (Variables|Default|GetbLoopingDownhillGCN)
    (Projectile|RemoveGameplayCue%28Looping%29Local (Variables|Default|GetAttachedPlayer) (Variables|Default|GetGCNTag_HighSpeed) (Ability|GameplayCue|MakeGameplayCueParameters 0.0))
    (else
      _))
  (Variables|Default|SetbLoopingDownhillGCN false))

(fn OnRep_AttachedPlayer)

(fn UpdateAudioandVFXParams
  (bind _speedfactor (LocalVariables|GetSpeedFactor))
  (bind _returnvalue (< _speedfactor 0.0))
  (bind _returnvalue_2 (> _speedfactor 0.0))
  (bind _attachedplayer (Variables|Default|GetAttachedPlayer))
  (bind _returnvalue_3 (Zipline|GetZiplineSpeedFactorTarget _attachedplayer))
  (bind _returnvalue_4 (< _returnvalue_3 0.0))
  (bind _returnvalue_5 (> _returnvalue_3 0.0))
  (bind _returnvalue_6 (and _returnvalue_2 _returnvalue_4))
  (bind _returnvalue_7 (and _returnvalue _returnvalue_5))
  (bind _returnvalue_8 (or _returnvalue_7 _returnvalue_6))
  (LocalVariables|SetSpeedFactor (Zipline|GetZiplineSpeedFactor (Variables|Default|GetAttachedPlayer)))
  (Utilities|FlowControl|Sequence
    (:then_0
      ;[1 1 1] Check if we're reversing momentum
      (if (Math|Boolean|NotEqual(Boolean) _returnvalue_8 (Variables|Default|GetIsReversingMomentum))
        (bind _output_get (Variables|Default|SetIsReversingMomentum _returnvalue_8))
        (if _output_get
          ;;[0 0.8562 1] Start/Stop grinding/breaking sound
          (Development|PrintString "Turn on grinding/breaking sound and start heavy sparks FX." (Variables|Default|GetdebugOutput) (Variables|Default|GetdebugOutput))
          ;;[1 0.1579 0.06819] Start/Stop heavy spark FX
          (Development|PrintString "START heavy sparks FX." (Variables|Default|GetdebugOutput) (Variables|Default|GetdebugOutput) "(R=1.000000,G=0.798898,B=0.026724,A=1.000000)")
          (GameplayCue|ExecuteGameplayCue%28Burst%29Local (Variables|Default|GetAttachedPlayer) (Variables|Default|GetGCNTag_ZiplineBraking) (Ability|GameplayCue|MakeGameplayCueParameters (Math|Float|Absolute(Float) (LocalVariables|GetSpeedFactor))))
          (else
            (Development|PrintString "Turn OFF grinding/breaking sound and STOP heavy sparks FX." (Variables|Default|GetdebugOutput) (Variables|Default|GetdebugOutput))
            (Development|PrintString "STOP heavy sparks FX." (Variables|Default|GetdebugOutput) (Variables|Default|GetdebugOutput) "(R=1.000000,G=0.566989,B=0.000000,A=1.000000)")))))
    (:then_1
      (if (> (LocalVariables|GetSpeedFactor) 0.0)
        ;[1 1 1] Traveling Uphill params
        (if (Variables|Default|GetbIsTravelingDownhill)
          (Variables|Default|SetbIsTravelingDownhill false)
          ;;[0 0.8562 1] Stop Audio for Downhill (bzzzzzzZZZRT)
          (Development|PrintString "Start Playing audio for traveling downhill (bzzzzZZZT)" (Variables|Default|GetdebugOutput) (Variables|Default|GetdebugOutput))
          ;;[1 0.1579 0.06819] Stop VFX for Downhill (speed lines)
          (Development|PrintString "Stop VFX for traveling downhill (speed lines)" (Variables|Default|GetdebugOutput) (Variables|Default|GetdebugOutput) "(R=1.000000,G=0.564712,B=0.014444,A=1.000000)")
          (CallFunction|CleanuponDownhillGCN))
        (if (not (Variables|Default|GetbIsTravelingUphill))
          (Variables|Default|SetbIsTravelingUphill true)
          ;;[0 0.8562 1] Start Audio for Uphill
          (Development|PrintString "Start Playing audio for traveling uphill (clicking)" (Variables|Default|GetdebugOutput) (Variables|Default|GetdebugOutput)))
        ;;[0 0.8562 1] Update Audio for uphill (clicking)
        (Development|PrintString (Utilities|String|Append "Update audio for traveling uphill (clicking) using " (Utilities|String|ToString(Float) (LocalVariables|GetSpeedFactor)) " to slow down the clicking as it reaches 1 (0 = means traveling horizontal.  1 means traveling straight up.).") (Variables|Default|GetdebugOutput) (Variables|Default|GetdebugOutput))
        (elif (Variables|Default|GetbIsTravelingUphill)
          ;[1 1 1] Traveling Downhill params
          (Variables|Default|SetbIsTravelingUphill false)
          ;;[0 0.8562 1] Stop Audio for Uphill (clicking)
          (Development|PrintString "STOP Playing audio for traveling uphill (clicking)" (Variables|Default|GetdebugOutput) (Variables|Default|GetdebugOutput))
          (if (not (Variables|Default|GetbIsTravelingDownhill))
            (Variables|Default|SetbIsTravelingDownhill true)
            ;;[0 0.8562 1] Start Audio for Downhill (bzzzzzzZZZRT)
            (Development|PrintString "Start Playing audio for traveling downhill (bzzzzZZZT)" (Variables|Default|GetdebugOutput) (Variables|Default|GetdebugOutput))
            ;;[1 0.1579 0.06819] Start VFX for Downhill (speed lines)
            (Development|PrintString "Start VFX for traveling downhill (speed lines)" (Variables|Default|GetdebugOutput) (Variables|Default|GetdebugOutput) "(R=1.000000,G=0.564712,B=0.014444,A=1.000000)")
            (Variables|Default|SetbLoopingDownhillGCN true)
            (Projectile|AddGameplayCue%28Looping%29Local (Variables|Default|GetAttachedPlayer) (Variables|Default|GetGCNTag_HighSpeed) (Ability|GameplayCue|MakeGameplayCueParameters (Math|Float|Absolute(Float) (LocalVariables|GetSpeedFactor)))))
          (LocalVariables|SetAlpha (* (LocalVariables|GetSpeedFactor) -1.0))
          ;;[0 0.8562 1] Update Audio for Downhill (pitch up as Alpha reaches 1)
          (Development|PrintString (Utilities|String|Append "Update/pitch up audio for traveling downhill (bzzzZZZrt) using " (Utilities|String|ToString(Float) (LocalVariables|GetAlpha)) ".  Pitch should grow higher as Alpha reaches 1.") (Variables|Default|GetdebugOutput) (Variables|Default|GetdebugOutput))
          ;;[1 0.1579 0.06819] Update Audio for Downhill (pitch up as Alpha reaches 1)
          (Development|PrintString (Utilities|String|Append "Update VFX for traveling downhill (speed lines) fade in as..." (Utilities|String|ToString(Float) (LocalVariables|GetAlpha)) "...reaches 1") (Variables|Default|GetdebugOutput) (Variables|Default|GetdebugOutput) "(R=1.000000,G=0.714649,B=0.000000,A=1.000000)")
          (else
            _))))))

(fn ConstructionScript)
