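; Yellow zipline harness Blueprint that drives speed-line and spark effects for its attached player.
; Two looping timers sample the player's mesh height, the slower one storing a baseline Z that the faster one compares against to detect upward movement.
; DoOnce-gated branches swap between Cascade and Niagara speed-line components.
; Alpha and Spark timelines feed particle parameters, and Spark updates also call ResetBlue to switch back to the Cascade lines.
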(event EventBeginPlay
  (|CallParentFunction)
  (Utilities|Time|SetTimerbyEvent (delegate Custom|Update%20Begin%20Z) 0.3 true)
  (Utilities|Time|SetTimerbyEvent (delegate Custom|UpdateZiplineFX) 0.1 true))

(event Custom|Update%20Begin%20Z
  (bind _attachedplayer (Variables|Default|GetAttachedPlayer))
  (Utilities|IsValid _attachedplayer
    (:Is%20Valid
      (Variables|Default|SetBeginLocation_z (.z (Transformation|GetWorldLocation (Variables|/Script/Engine.Character|GetMesh _attachedplayer)))))))

(event Custom|UpdateZiplineFX
  (bind _attachedplayer (Variables|Default|GetAttachedPlayer))
  (Utilities|IsValid _attachedplayer
    (:Is%20Valid
      (Variables|Default|SetLocation (.z (Transformation|GetWorldLocation (Variables|/Script/Engine.Character|GetMesh _attachedplayer))))
      (if (> (Variables|Default|GetLocation) (Variables|Default|GetBeginLocation_z))
        (Utilities|FlowControl|Sequence
          (:then_0
            (Utilities|FlowControl|DoOnce false
              (:Completed
                (Components|Activation|Activate (Variables|Default|GetP_Zipline_Pulley_SpeedLines) true)
                (Components|Activation|Deactivate (Variables|Default|GetNS_Zipline_Speedline))
                (bind (_direction _newtrack_0) (Timeline|Alpha PlayFromStart)
                  (:Update
                    (Effects|Components|ParticleSystem|SetFloatParameter (Variables|Default|GetP_Zipline_Pulley_SpeedLines) "Alpha" _newtrack_0))))))
          (:then_1
            _))
        (else
          (Utilities|FlowControl|Sequence
            (:then_0
              _))))
      (Utilities|FlowControl|DoOnce false
        (:Completed
          (Components|Activation|Deactivate (Variables|Default|GetP_Zipline_Pulley_SpeedLines))
          (Components|Activation|Activate (Variables|Default|GetNS_Zipline_Speedline) false)
          (Components|Activation|Activate (Variables|Default|GetNS_Zipline_Pulley_SpeedLines_Converted) false)
          (bind (_direction_2 _newtrack_0_2) (Timeline|Spark PlayFromStart)
            (:Update
              (Niagara|SetNiagaraVariable(Float) (Variables|Default|GetNS_Zipline_Pulley_SpeedLines_Converted) "SpawnSpark" _newtrack_0_2)
              (CallFunction|ResetBlue))))))))

(event Custom|ResetBlue
  (Utilities|FlowControl|DoOnce false
    (:Completed
      (Components|Activation|Activate (Variables|Default|GetP_Zipline_Pulley_SpeedLines) true)
      (Components|Activation|Deactivate (Variables|Default|GetNS_Zipline_Speedline))
      (Timeline|Alpha PlayFromStart))))

(fn ConstructionScript
  (|CallParentFunction))
