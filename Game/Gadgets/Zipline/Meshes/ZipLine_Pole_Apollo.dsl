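; Zipline pole building actor that lets recently ziplining players pass through it instead of colliding, and optionally dies when its linked foundation is destroyed.
; On BeginPlay it spawns a companion capsule actor used for overlap checks.
; While pawns overlap the large capsule, zipline state changes and a looping timer track last-zipline times, using small-capsule overlaps to manage per-pawn ignores or optionally disable mesh collision.
; A helper box-traces downward to add editor-only tags to the foundation.
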(event EventBeginPlay
  ;[1 1 1] Spawn collision actor for capsule overlaps
  (bind _returnvalue (Building|SpawnBuildingActor "/Game/Gadgets/Zipline/Meshes/ZipLine_Pole_Capsules.ZipLine_Pole_Capsules_C" (Transformation|GetActorTransform) 0 0 false true))
  (bind _aszip_line_pole_capsules (Utilities|Casting|CastToZipLine_Pole_Capsules _returnvalue))
  (Variables|/Game/Gadgets/Zipline/Meshes/ZipLine_Pole_Capsules.ZipLine_Pole_Capsules_C|SetZiplinePole self _aszip_line_pole_capsules)
  (Variables|Default|SetZiplineCapulseActor _aszip_line_pole_capsules)
  ;[0.4072 0.7682 1] Debug prints when pawns collide
  (if (Variables|Default|GetDebug)
    (Collision|BindEventtoOnActorHit (delegate Custom|Actor%20Hit))))

(event Custom|Actor%20Hit
  (params (SelfActor :type Actor%20Object%20Reference) (OtherActor :type Actor%20Object%20Reference) (NormalImpulse :type Vector) (Hit :type Hit%20Result%20Structure%20%28by%20ref%29))
  (bind _pawnsoverlappinglarge (Variables|Default|GetPawnsOverlappingLarge))
  ;[0.4072 0.7682 1] Debug prints when pawns collide
  (bind _asfort_player_pawn (Utilities|Casting|CastToFortPlayerPawn OtherActor))
  (bind _index (CallFunction|FindinIgnorePlayers _asfort_player_pawn))
  (if (!= _index -1)
    (bind (_output_fortplayerpawn_2_5d98175c48c7530a6d858f9372994086 _output_lastziplinetime_9_fc1293f34fc6624424dcac956b88ca00) (Utilities|Struct|BreakIgnorePlayerStruct (Utilities|Array|Get(aref) (Variables|Default|GetPlayersIgnoringPole) _index)))
    (Development|PrintString (Utilities|String|Append "HIt Pawn arrayID=" (Utilities|String|ToString(Integer) (Utilities|Array|FindItem _pawnsoverlappinglarge _asfort_player_pawn)) "  Last ziplining X Seconds Ago." (Utilities|String|ToString(Float) (- (Utilities|Time|GetTimeSeconds) _output_lastziplinetime_9_fc1293f34fc6624424dcac956b88ca00)) "    1 of " (Utilities|String|ToString(Integer) (Utilities|Array|Length (Variables|Default|GetPlayersIgnoringPole)))) (Variables|Default|GetDebug) (Variables|Default|GetDebug) "(R=1.000000,G=0.000000,B=0.086574,A=1.000000)" 2e+01)
    (else
      (Development|PrintString (Utilities|String|Append "HIt Pawn arrayID=" (Utilities|String|ToString(Integer) (Utilities|Array|FindItem _pawnsoverlappinglarge _asfort_player_pawn)) "") (Variables|Default|GetDebug) (Variables|Default|GetDebug)))))

(event Custom|PlayerBeginOverlappingLargeCapsule
  (params (FortPlayerPawn :type Fort%20Player%20Pawn%20Object%20Reference))
  (bind _returnvalue (Utilities|Time|GetTimeSeconds))
  (bind _output_fortplayerpawn_2_5d98175c48c7530a6d858f9372994086 (Utilities|Struct|BreakIgnorePlayerStruct (Utilities|Array|Get(aref) (Variables|Default|GetPlayersIgnoringPole) (Variables|Default|GetPlayersIgnoringPoleIndex))))
  ;[1 1 1] When we're overlapping the large capsule, add our pawn and bind necessary and check ziplinig state
  (Utilities|Array|AddUnique (Variables|Default|GetPawnsOverlappingLarge) FortPlayerPawn)
  (EventDispatchers|BindEventtoZiplineStateChanged FortPlayerPawn (delegate Custom|pawnZiplineStateChanged))
  (bind _returnvalue_2 (Zipline|IsZiplining FortPlayerPawn))
  (if _returnvalue_2
    ;[1 1 1] Handle zipline state with looping timer
    (if (not (Utilities|Time|IsTimerActivebyHandle (Variables|Default|GetOverlapTimer)))
      (bind _returnvalue_3 (Utilities|Time|SetTimerbyEvent (delegate Custom|updateZiplinePoleCollisionWithOverlappingPawns) (Variables|Default|GetTimerSecondsInterval) true))
      (Variables|Default|SetOverlapTimer _returnvalue_3)
      (bind (_arrayindex _arrayelement) (Utilities|Array|ReverseforEachLoop (Variables|Default|GetPawnsOverlappingLarge))
        (:LoopBody
          (bind _output_get (Variables|Default|SetCurFortPlayerPawn _arrayelement))
          (bind _returnvalue_4 (Zipline|IsZiplining _output_get))
          (if _returnvalue_4
            ;[1 1 1] When ziplining reset out time ziplining
            (bind _index (CallFunction|FindinIgnorePlayers (Variables|Default|GetCurFortPlayerPawn)))
            (Variables|Default|SetPlayersIgnoringPoleIndex _index)
            (if (== (Variables|Default|GetPlayersIgnoringPoleIndex) -1)
              (bind _returnvalue_5 (Utilities|Array|Add (Variables|Default|GetPlayersIgnoringPole) (Utilities|Struct|MakeIgnorePlayerStruct (Variables|Default|GetCurFortPlayerPawn) (Utilities|Time|GetTimeSeconds))))
              (Variables|Default|SetPlayersIgnoringPoleIndex _returnvalue_5)
              (else
                (Utilities|Array|SetArrayElem (Variables|Default|GetPlayersIgnoringPole) (Variables|Default|GetPlayersIgnoringPoleIndex) (Utilities|Struct|MakeIgnorePlayerStruct _output_fortplayerpawn_2_5d98175c48c7530a6d858f9372994086 (Utilities|Time|GetTimeSeconds)) false)))
            ;[0.4072 0.7682 1] Debug line
            (if (Variables|Default|GetDebug)
              (Rendering|Debug|DrawDebugLine (Transformation|GetActorLocation (Variables|Default|GetCurFortPlayerPawn)) (+ (Transformation|GetActorLocation (Variables|Default|GetCurFortPlayerPawn)) "0, 0,150.000000") "(R=0.000000,G=1.000000,B=0.358541,A=1.000000)" (Variables|Default|GetIgnorePoleCollisionForXSeconds) 4.0)
              (Development|PrintString (Utilities|String|Append "Resetting zipline time.  " (Utilities|String|ToString(Integer) (Utilities|Array|Length (Variables|Default|GetPlayersIgnoringPole))) "    Timer Count" (Utilities|String|ToString(Integer) (Variables|Default|GetTimerCount))) (Variables|Default|GetDebug) (Variables|Default|GetDebug)))))
        (:Completed
          (bind (_arrayindex_2 _arrayelement_2) (Utilities|Array|ReverseforEachLoop (Variables|Default|GetPlayersIgnoringPole))
            (:LoopBody
              (Variables|Default|SetPlayersIgnoringPoleIndex _arrayindex_2)
              (bind (_arrayelement_fortplayerpawn_2_5d98175c48c7530a6d858f9372994086 _arrayelement_lastziplinetime_9_fc1293f34fc6624424dcac956b88ca00) (Utilities|Struct|BreakIgnorePlayerStruct _arrayelement_2))
              (Variables|Default|SetCurFortPlayerPawn _arrayelement_fortplayerpawn_2_5d98175c48c7530a6d858f9372994086)
              (if (not (<= (- _returnvalue _arrayelement_lastziplinetime_9_fc1293f34fc6624424dcac956b88ca00) (Variables|Default|GetIgnorePoleCollisionForXSeconds)))
                ;[0.4072 0.7682 1] Comment
                (Development|PrintString (Utilities|String|Append "Seconds since ziplining:" (Utilities|String|ToString(Float) (- _returnvalue _arrayelement_lastziplinetime_9_fc1293f34fc6624424dcac956b88ca00)) "    TimerCount:" (Utilities|String|ToString(Integer) (Variables|Default|GetTimerCount))) (Variables|Default|GetDebug) (Variables|Default|GetDebug) "(R=1.000000,G=0.054146,B=0.316445,A=1.000000)" 6.0)
                ;[1 1 1] If we're overlapping the smaller capsules...
                (bind _ziplinecapulseactor (Variables|Default|GetZiplineCapulseActor)
                  (:then
                    ;[1 1 1] Remove Ignoring Pawn
                    (if (or (Collision|IsOverlappingActor (Variables|/Game/Gadgets/Zipline/Meshes/ZipLine_Pole_Capsules.ZipLine_Pole_Capsules_C|GetCapsuleSmall _ziplinecapulseactor) (Variables|Default|GetCurFortPlayerPawn) :!class /Script/Engine.PrimitiveComponent) (Collision|IsOverlappingActor (Variables|/Game/Gadgets/Zipline/Meshes/ZipLine_Pole_Capsules.ZipLine_Pole_Capsules_C|GetCapsuleFittedTop _ziplinecapulseactor) (Variables|Default|GetCurFortPlayerPawn) :!class /Script/Engine.PrimitiveComponent))
                      ;[0.4072 0.7682 1] Comment
                      (Development|PrintString "overlapping small collision" (Variables|Default|GetDebug) (Variables|Default|GetDebug))
                      (if (Variables|Default|GetDisableCollisionCompletely) :!label _label
                        (Collision|SetCollisionEnabled (Variables|Debug|GetStaticMeshComponent) "NoCollision"))
                      (Collision|AddIgnoredActor (Variables|Default|GetCurFortPlayerPawn) self (Variables|Default|GetIgnorePoleCollisionForXSeconds))
                      (Class|ZipLinePoleCapsules|SetSmallOverlapCollisionEnabled (Variables|Default|GetZiplineCapulseActor) true)
                      ;[0.4072 0.7682 1] Debug
                      (if (Variables|Default|GetDebug)
                        (Rendering|Debug|DrawDebugLine (Transformation|GetActorLocation) (+ (Transformation|GetActorLocation) "0, 0,1000.000000") "(R=1.000000,G=0.000000,B=0.027873,A=1.000000)" (Variables|Default|GetTimerSecondsInterval) 5.0)
                        (Development|SetHiddeninGame (Variables|Debug|GetStaticMeshComponent) true false))
                      (else
                        _)))
                  (:else
                    _))
                (CallFunction|RemoveIgnorePlayer (Variables|Default|GetCurFortPlayerPawn))))
            (:Completed
              (Math|Integer|IncrementInt (Variables|Default|GetTimerCount))))))
      (goto _label))))

(event Custom|pawnZiplineStateChanged
  (params (bIsZiplining :type Boolean) (FortPlayerPawn :type Fort%20Player%20Pawn%20Object%20Reference))
  (bind _returnvalue (Utilities|Time|GetTimeSeconds))
  (bind _output_fortplayerpawn_2_5d98175c48c7530a6d858f9372994086 (Utilities|Struct|BreakIgnorePlayerStruct (Utilities|Array|Get(aref) (Variables|Default|GetPlayersIgnoringPole) (Variables|Default|GetPlayersIgnoringPoleIndex))))
  ;[1 1 1] Handle zipline state with looping timer
  (if bIsZiplining
    (if (not (Utilities|Time|IsTimerActivebyHandle (Variables|Default|GetOverlapTimer)))
      (bind _returnvalue_2 (Utilities|Time|SetTimerbyEvent (delegate Custom|updateZiplinePoleCollisionWithOverlappingPawns) (Variables|Default|GetTimerSecondsInterval) true))
      (Variables|Default|SetOverlapTimer _returnvalue_2)
      (bind (_arrayindex _arrayelement) (Utilities|Array|ReverseforEachLoop (Variables|Default|GetPawnsOverlappingLarge))
        (:LoopBody
          (bind _output_get (Variables|Default|SetCurFortPlayerPawn _arrayelement))
          (bind _returnvalue_3 (Zipline|IsZiplining _output_get))
          (if _returnvalue_3
            ;[1 1 1] When ziplining reset out time ziplining
            (bind _index (CallFunction|FindinIgnorePlayers (Variables|Default|GetCurFortPlayerPawn)))
            (Variables|Default|SetPlayersIgnoringPoleIndex _index)
            (if (== (Variables|Default|GetPlayersIgnoringPoleIndex) -1)
              (bind _returnvalue_4 (Utilities|Array|Add (Variables|Default|GetPlayersIgnoringPole) (Utilities|Struct|MakeIgnorePlayerStruct (Variables|Default|GetCurFortPlayerPawn) (Utilities|Time|GetTimeSeconds))))
              (Variables|Default|SetPlayersIgnoringPoleIndex _returnvalue_4)
              (else
                (Utilities|Array|SetArrayElem (Variables|Default|GetPlayersIgnoringPole) (Variables|Default|GetPlayersIgnoringPoleIndex) (Utilities|Struct|MakeIgnorePlayerStruct _output_fortplayerpawn_2_5d98175c48c7530a6d858f9372994086 (Utilities|Time|GetTimeSeconds)) false)))
            ;[0.4072 0.7682 1] Debug line
            (if (Variables|Default|GetDebug)
              (Rendering|Debug|DrawDebugLine (Transformation|GetActorLocation (Variables|Default|GetCurFortPlayerPawn)) (+ (Transformation|GetActorLocation (Variables|Default|GetCurFortPlayerPawn)) "0, 0,150.000000") "(R=0.000000,G=1.000000,B=0.358541,A=1.000000)" (Variables|Default|GetIgnorePoleCollisionForXSeconds) 4.0)
              (Development|PrintString (Utilities|String|Append "Resetting zipline time.  " (Utilities|String|ToString(Integer) (Utilities|Array|Length (Variables|Default|GetPlayersIgnoringPole))) "    Timer Count" (Utilities|String|ToString(Integer) (Variables|Default|GetTimerCount))) (Variables|Default|GetDebug) (Variables|Default|GetDebug)))))
        (:Completed
          (bind (_arrayindex_2 _arrayelement_2) (Utilities|Array|ReverseforEachLoop (Variables|Default|GetPlayersIgnoringPole))
            (:LoopBody
              (Variables|Default|SetPlayersIgnoringPoleIndex _arrayindex_2)
              (bind (_arrayelement_fortplayerpawn_2_5d98175c48c7530a6d858f9372994086 _arrayelement_lastziplinetime_9_fc1293f34fc6624424dcac956b88ca00) (Utilities|Struct|BreakIgnorePlayerStruct _arrayelement_2))
              (Variables|Default|SetCurFortPlayerPawn _arrayelement_fortplayerpawn_2_5d98175c48c7530a6d858f9372994086)
              (if (not (<= (- _returnvalue _arrayelement_lastziplinetime_9_fc1293f34fc6624424dcac956b88ca00) (Variables|Default|GetIgnorePoleCollisionForXSeconds)))
                ;[0.4072 0.7682 1] Comment
                (Development|PrintString (Utilities|String|Append "Seconds since ziplining:" (Utilities|String|ToString(Float) (- _returnvalue _arrayelement_lastziplinetime_9_fc1293f34fc6624424dcac956b88ca00)) "    TimerCount:" (Utilities|String|ToString(Integer) (Variables|Default|GetTimerCount))) (Variables|Default|GetDebug) (Variables|Default|GetDebug) "(R=1.000000,G=0.054146,B=0.316445,A=1.000000)" 6.0)
                ;[1 1 1] If we're overlapping the smaller capsules...
                (bind _ziplinecapulseactor (Variables|Default|GetZiplineCapulseActor)
                  (:then
                    ;[1 1 1] Remove Ignoring Pawn
                    (if (or (Collision|IsOverlappingActor (Variables|/Game/Gadgets/Zipline/Meshes/ZipLine_Pole_Capsules.ZipLine_Pole_Capsules_C|GetCapsuleSmall _ziplinecapulseactor) (Variables|Default|GetCurFortPlayerPawn) :!class /Script/Engine.PrimitiveComponent) (Collision|IsOverlappingActor (Variables|/Game/Gadgets/Zipline/Meshes/ZipLine_Pole_Capsules.ZipLine_Pole_Capsules_C|GetCapsuleFittedTop _ziplinecapulseactor) (Variables|Default|GetCurFortPlayerPawn) :!class /Script/Engine.PrimitiveComponent))
                      ;[0.4072 0.7682 1] Comment
                      (Development|PrintString "overlapping small collision" (Variables|Default|GetDebug) (Variables|Default|GetDebug))
                      (if (Variables|Default|GetDisableCollisionCompletely) :!label _label
                        (Collision|SetCollisionEnabled (Variables|Debug|GetStaticMeshComponent) "NoCollision"))
                      (Collision|AddIgnoredActor (Variables|Default|GetCurFortPlayerPawn) self (Variables|Default|GetIgnorePoleCollisionForXSeconds))
                      (Class|ZipLinePoleCapsules|SetSmallOverlapCollisionEnabled (Variables|Default|GetZiplineCapulseActor) true)
                      ;[0.4072 0.7682 1] Debug
                      (if (Variables|Default|GetDebug)
                        (Rendering|Debug|DrawDebugLine (Transformation|GetActorLocation) (+ (Transformation|GetActorLocation) "0, 0,1000.000000") "(R=1.000000,G=0.000000,B=0.027873,A=1.000000)" (Variables|Default|GetTimerSecondsInterval) 5.0)
                        (Development|SetHiddeninGame (Variables|Debug|GetStaticMeshComponent) true false))
                      (else
                        _)))
                  (:else
                    _))
                (CallFunction|RemoveIgnorePlayer (Variables|Default|GetCurFortPlayerPawn))))
            (:Completed
              (Math|Integer|IncrementInt (Variables|Default|GetTimerCount))))))
      (goto _label))))

(event Custom|updateZiplinePoleCollisionWithOverlappingPawns
  (bind _returnvalue (Utilities|Time|GetTimeSeconds))
  (bind _output_fortplayerpawn_2_5d98175c48c7530a6d858f9372994086 (Utilities|Struct|BreakIgnorePlayerStruct (Utilities|Array|Get(aref) (Variables|Default|GetPlayersIgnoringPole) (Variables|Default|GetPlayersIgnoringPoleIndex))))
  ;[1 1 1] Handle zipline state with looping timer
  (bind (_arrayindex _arrayelement) (Utilities|Array|ReverseforEachLoop (Variables|Default|GetPawnsOverlappingLarge))
    (:LoopBody
      (bind _output_get (Variables|Default|SetCurFortPlayerPawn _arrayelement))
      (bind _returnvalue_2 (Zipline|IsZiplining _output_get))
      (if _returnvalue_2
        ;[1 1 1] When ziplining reset out time ziplining
        (bind _index (CallFunction|FindinIgnorePlayers (Variables|Default|GetCurFortPlayerPawn)))
        (Variables|Default|SetPlayersIgnoringPoleIndex _index)
        (if (== (Variables|Default|GetPlayersIgnoringPoleIndex) -1)
          (bind _returnvalue_3 (Utilities|Array|Add (Variables|Default|GetPlayersIgnoringPole) (Utilities|Struct|MakeIgnorePlayerStruct (Variables|Default|GetCurFortPlayerPawn) (Utilities|Time|GetTimeSeconds))))
          (Variables|Default|SetPlayersIgnoringPoleIndex _returnvalue_3)
          (else
            (Utilities|Array|SetArrayElem (Variables|Default|GetPlayersIgnoringPole) (Variables|Default|GetPlayersIgnoringPoleIndex) (Utilities|Struct|MakeIgnorePlayerStruct _output_fortplayerpawn_2_5d98175c48c7530a6d858f9372994086 (Utilities|Time|GetTimeSeconds)) false)))
        ;[0.4072 0.7682 1] Debug line
        (if (Variables|Default|GetDebug)
          (Rendering|Debug|DrawDebugLine (Transformation|GetActorLocation (Variables|Default|GetCurFortPlayerPawn)) (+ (Transformation|GetActorLocation (Variables|Default|GetCurFortPlayerPawn)) "0, 0,150.000000") "(R=0.000000,G=1.000000,B=0.358541,A=1.000000)" (Variables|Default|GetIgnorePoleCollisionForXSeconds) 4.0)
          (Development|PrintString (Utilities|String|Append "Resetting zipline time.  " (Utilities|String|ToString(Integer) (Utilities|Array|Length (Variables|Default|GetPlayersIgnoringPole))) "    Timer Count" (Utilities|String|ToString(Integer) (Variables|Default|GetTimerCount))) (Variables|Default|GetDebug) (Variables|Default|GetDebug)))))
    (:Completed
      (bind (_arrayindex_2 _arrayelement_2) (Utilities|Array|ReverseforEachLoop (Variables|Default|GetPlayersIgnoringPole))
        (:LoopBody
          (Variables|Default|SetPlayersIgnoringPoleIndex _arrayindex_2)
          (bind (_arrayelement_fortplayerpawn_2_5d98175c48c7530a6d858f9372994086 _arrayelement_lastziplinetime_9_fc1293f34fc6624424dcac956b88ca00) (Utilities|Struct|BreakIgnorePlayerStruct _arrayelement_2))
          (Variables|Default|SetCurFortPlayerPawn _arrayelement_fortplayerpawn_2_5d98175c48c7530a6d858f9372994086)
          (if (not (<= (- _returnvalue _arrayelement_lastziplinetime_9_fc1293f34fc6624424dcac956b88ca00) (Variables|Default|GetIgnorePoleCollisionForXSeconds)))
            ;[0.4072 0.7682 1] Comment
            (Development|PrintString (Utilities|String|Append "Seconds since ziplining:" (Utilities|String|ToString(Float) (- _returnvalue _arrayelement_lastziplinetime_9_fc1293f34fc6624424dcac956b88ca00)) "    TimerCount:" (Utilities|String|ToString(Integer) (Variables|Default|GetTimerCount))) (Variables|Default|GetDebug) (Variables|Default|GetDebug) "(R=1.000000,G=0.054146,B=0.316445,A=1.000000)" 6.0)
            ;[1 1 1] If we're overlapping the smaller capsules...
            (bind _ziplinecapulseactor (Variables|Default|GetZiplineCapulseActor)
              (:then
                ;[1 1 1] Remove Ignoring Pawn
                (if (or (Collision|IsOverlappingActor (Variables|/Game/Gadgets/Zipline/Meshes/ZipLine_Pole_Capsules.ZipLine_Pole_Capsules_C|GetCapsuleSmall _ziplinecapulseactor) (Variables|Default|GetCurFortPlayerPawn) :!class /Script/Engine.PrimitiveComponent) (Collision|IsOverlappingActor (Variables|/Game/Gadgets/Zipline/Meshes/ZipLine_Pole_Capsules.ZipLine_Pole_Capsules_C|GetCapsuleFittedTop _ziplinecapulseactor) (Variables|Default|GetCurFortPlayerPawn) :!class /Script/Engine.PrimitiveComponent))
                  ;[0.4072 0.7682 1] Comment
                  (Development|PrintString "overlapping small collision" (Variables|Default|GetDebug) (Variables|Default|GetDebug))
                  (if (Variables|Default|GetDisableCollisionCompletely) :!label _label
                    (Collision|SetCollisionEnabled (Variables|Debug|GetStaticMeshComponent) "NoCollision"))
                  (Collision|AddIgnoredActor (Variables|Default|GetCurFortPlayerPawn) self (Variables|Default|GetIgnorePoleCollisionForXSeconds))
                  (Class|ZipLinePoleCapsules|SetSmallOverlapCollisionEnabled (Variables|Default|GetZiplineCapulseActor) true)
                  ;[0.4072 0.7682 1] Debug
                  (if (Variables|Default|GetDebug)
                    (Rendering|Debug|DrawDebugLine (Transformation|GetActorLocation) (+ (Transformation|GetActorLocation) "0, 0,1000.000000") "(R=1.000000,G=0.000000,B=0.027873,A=1.000000)" (Variables|Default|GetTimerSecondsInterval) 5.0)
                    (Development|SetHiddeninGame (Variables|Debug|GetStaticMeshComponent) true false))
                  (else
                    _)))
              (:else
                _))
            (CallFunction|RemoveIgnorePlayer (Variables|Default|GetCurFortPlayerPawn))))
        (:Completed
          (Math|Integer|IncrementInt (Variables|Default|GetTimerCount))))))
  (goto _label))

(event Custom|UnbindZiplineStateChanged
  (params (InFortPlayerPawn :type Fort%20Player%20Pawn%20Object%20Reference))
  ;[1 1 1] Unbind zipline state changed
  (EventDispatchers|UnbindEventfromZiplineStateChanged InFortPlayerPawn (delegate Custom|pawnZiplineStateChanged)))

(event Custom|PlayerEndOverlappingLargeCapsule
  (params (FortPlayerPawn :type Fort%20Player%20Pawn%20Object%20Reference))
  ;[1 1 1] End overlap remove pawn from our list
  (CallFunction|RemovePawnfromPawnsOverlappingLarge FortPlayerPawn))

(event OnLinkedActorDestroyed(FortLinkToActor)
  (params (DamageCauser :type Actor%20Object%20Reference))
  ;[1 1 1] Die when foundation has been destroyed (if DestructionEnabled is true)
  (if (ScalableFloat|GetValueasBool (Variables|Default|GetDestructionEnabled))
    (Gameplay|Die :DamageCauser DamageCauser)))

(fn AddZiplineSmashImmuneTagToFoundation
  (bind _outhit (Collision|BoxTraceByChannel (Transformation|GetActorLocation) (+ (Transformation|GetActorLocation) (* (Math|Vector|GetUpVector "0, 0, 0") -5e+01)) "20.000000,20.000000,20.000000" "0, 0, 0" "TraceTypeQuery10"))
  (bind _asbuilding_actor (Utilities|Casting|CastToBuildingActor (.HitActor (Collision|BreakHitResult _outhit))))
  (Tags|AddEditorOnlyInstanceGameplayTags _asbuilding_actor (LocalVariables|GetTagsToAddToFoundation)))

(fn FindInIgnorePlayers
  (params (InFortPlayerPawn :type Fort%20Player%20Pawn%20Object%20Reference))
  (LocalVariables|SetFoundIndex -1)
  (for (_array_element _array_index) (Variables|Default|GetPlayersIgnoringPole)
    (bind _array_element_fortplayerpawn_2_5d98175c48c7530a6d858f9372994086 (Utilities|Struct|BreakIgnorePlayerStruct _array_element))
    (if (Utilities|Equal(Object) _array_element_fortplayerpawn_2_5d98175c48c7530a6d858f9372994086 (Variables|GetInFortPlayerPawn))
      (LocalVariables|SetFoundIndex _array_index)))
  (return (LocalVariables|GetFoundIndex)))

(fn RemoveIgnorePlayer
  (params (InFortPlayerPawn :type Fort%20Player%20Pawn%20Object%20Reference))
  (bind _index (CallFunction|FindinIgnorePlayers InFortPlayerPawn))
  (bind _output_get (Variables|Default|SetPlayersIgnoringPoleIndex _index))
  (if (!= _output_get -1)
    (Collision|AddIgnoredActor InFortPlayerPawn self 0.0)
    (Utilities|Array|RemoveIndex (Variables|Default|GetPlayersIgnoringPole) (Variables|Default|GetPlayersIgnoringPoleIndex))
    ;[0.4072 0.7682 1] debug
    (if (Variables|Default|GetDebug)
      (Development|SetHiddeninGame (Variables|Debug|GetStaticMeshComponent) false false))))

(fn RemovePawnFromPawnsOverlappingLarge
  (params (playerPawn :type Fort%20Player%20Pawn%20Object%20Reference))
  (CallFunction|RemoveIgnorePlayer playerPawn)
  (if (!= (Utilities|Array|FindItem (Variables|Default|GetPawnsOverlappingLarge) playerPawn) -1)
    (Utilities|Array|RemoveItem (Variables|Default|GetPawnsOverlappingLarge) playerPawn)
    (CallFunction|UnbindZiplineStateChanged playerPawn))
  (if (<= (Utilities|Array|Length (Variables|Default|GetPawnsOverlappingLarge)) 0)
    (Utilities|Time|ClearandInvalidateTimerbyHandle (Variables|Default|GetOverlapTimer))
    (Class|ZipLinePoleCapsules|SetSmallOverlapCollisionEnabled (Variables|Default|GetZiplineCapulseActor) false)
    ;[1 1 1] Check Disable/Enable Collision Completely
    (if (Variables|Default|GetDisableCollisionCompletely)
      (Collision|SetCollisionEnabled (Variables|Debug|GetStaticMeshComponent) "QueryAndPhysics"))))

(fn ConstructionScript)
