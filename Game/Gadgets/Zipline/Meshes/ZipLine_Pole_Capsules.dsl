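; Collision capsule helper for the Apollo zipline pole that reports player overlaps to its referenced pole.
; At BeginPlay it binds the large capsule's overlap events, casts the other actor to FortPlayerPawn and forwards begin and end to the pole.
; Small and fitted-top capsules start without collision and are toggled by SetSmallOverlapCollisionEnabled.
; When the pole's Debug flag is on, that toggle also shows or hides the fitted-top and large capsules.
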(event EventBeginPlay
  (CallFunction|BindLargeCapsuleOverlapEvents))

(event Custom|BindLargeCapsuleOverlapEvents
  (Collision|BindEventtoOnComponentBeginOverlap (Variables|Default|GetCapsuleLarge) (delegate Custom|BeginLargeCapsuleOverlap))
  (Collision|BindEventtoOnComponentEndOverlap (Variables|Default|GetCapsuleLarge) (delegate Custom|EndLargeCapsuleOverlap)))

(event Custom|BeginLargeCapsuleOverlap
  (params (OverlappedComponent :type Primitive%20Component%20Object%20Reference) (OtherActor :type Actor%20Object%20Reference) (OtherComp :type Primitive%20Component%20Object%20Reference) (OtherBodyIndex :type Integer) (bFromSweep :type Boolean) (SweepResult :type Hit%20Result%20Structure%20%28by%20ref%29))
  (bind _asfort_player_pawn (Utilities|Casting|CastToFortPlayerPawn OtherActor))
  (Class|ZipLinePoleApollo|PlayerBeginOverlappingLargeCapsule (Variables|Default|GetZiplinePole) _asfort_player_pawn))

(event Custom|EndLargeCapsuleOverlap
  (params (OverlappedComponent :type Primitive%20Component%20Object%20Reference) (OtherActor :type Actor%20Object%20Reference) (OtherComp :type Primitive%20Component%20Object%20Reference) (OtherBodyIndex :type Integer))
  (bind _asfort_player_pawn (Utilities|Casting|CastToFortPlayerPawn OtherActor))
  (Class|ZipLinePoleApollo|PlayerEndOverlappingLargeCapsule (Variables|Default|GetZiplinePole) _asfort_player_pawn))

(fn SetSmallOverlapCollisionEnabled
  (params (SetTo :type Boolean))
  (bind _returnvalue (not SetTo))
  (if SetTo
    (Collision|SetCollisionEnabled (Variables|Default|GetCapsuleSmall) "QueryAndPhysics")
    (Collision|SetCollisionEnabled (Variables|Default|GetCapsuleFittedTop) "QueryAndPhysics")
    (else
      (Collision|SetCollisionEnabled (Variables|Default|GetCapsuleSmall) "NoCollision")
      (Collision|SetCollisionEnabled (Variables|Default|GetCapsuleFittedTop) "NoCollision")))
  ;[1 1 1] Debug
  (if (Variables|/Game/Gadgets/Zipline/Meshes/ZipLine_Pole_Apollo.ZipLine_Pole_Apollo_C|GetDebug (Variables|Default|GetZiplinePole))
    (Development|SetHiddeninGame (Variables|Default|GetCapsuleFittedTop) _returnvalue false)
    (Development|SetHiddeninGame (Variables|Default|GetCapsuleLarge) _returnvalue false)))

(fn ConstructionScript
  (Collision|SetCollisionEnabled (Variables|Default|GetCapsuleSmall) "NoCollision")
  (Collision|SetCollisionEnabled (Variables|Default|GetCapsuleFittedTop) "NoCollision"))
